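# Sound cue for the looping capture sound of the LabradorRiding capture device.
# A looping Creative capture-area power-up wave feeds a continuous modulator, letting a Pitch parameter drive playback pitch within a narrow clamped range around normal.
# The cue plays at reduced volume and references a shared Fortnite trap-activation attenuation asset for distance falloff.
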SoundCue /LabradorRidingContent/Assets/Sound/LabradorRiding_CaptureDevice_Capture_Loop_Cue
root=ModulatorContinuous_0
nodes [
 {
  id=ModulatorContinuous_0
  children [WavePlayer_0]
  properties {pitchModulationParams {maxInput=1.2, maxOutput=1.2, minInput=0.9, minOutput=0.9, parameterName=Pitch}}
 }
 {id=WavePlayer_0, properties {bLooping=true, soundWaveAssetPtr=/Game/Sounds/Creative/Toys/CaptureDevice/CaptureArea_04_PowerUp_Loop}}
]
properties {
 attenuationSettings=/Game/Sounds/Attenuation/Fort_Traps_Activate_Athena_Att
 maxDistance=3500
 volumeMultiplier=0.7
 attenuationOverrides {bSpatialize=false, falloffDistance=0, attenuationShapeExtents {x=200}}
}
